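# Physics asset for the female medium Skull Brite VDay player body, simulating a skull accessory and its two bows.
# Box bodies cover a small driver bone, the skull, and left and right bow bones.
# Constraints chain the skull to the driver with tight swing and tighter twist limits, and each bow to the skull with looser limits, allowing secondary jiggle motion.
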PhysicsAsset /Game/Characters/Player/Female/Medium/Bodies/F_MED_Skull_Brite_VDay/Meshes/F_MED_Skull_Brite_VDay_Physics
bodies {
 driver_dyn_skull [{shapeType=Box, extentX=1.01, extentY=1.01, extentZ=1.01, rotation=<rot 0 -90 0>}]
 dyn_bow_l [{shapeType=Box, extentX=5, extentY=1.5, extentZ=5, center=<1.1244 0 0>}]
 dyn_bow_r [{shapeType=Box, extentX=5, extentY=1.5, extentZ=5, center=<0.9191 0 0>}]
 dyn_skull [{shapeType=Box, extentX=2.8842, extentY=14.5482, extentZ=16.0687, center=<0 -3.6801 -0.0004>}]
}
constraints [
 {bone1Name=dyn_skull, bone2Name=driver_dyn_skull, swing1LimitDegrees=20, swing2LimitDegrees=20, twistLimitDegrees=5}
 {bone1Name=dyn_bow_l, bone2Name=dyn_skull, swing1LimitDegrees=45, swing2LimitDegrees=45, twistLimitDegrees=45}
 {bone1Name=dyn_bow_r, bone2Name=dyn_skull, swing1LimitDegrees=45, swing2LimitDegrees=45, twistLimitDegrees=45}
]
