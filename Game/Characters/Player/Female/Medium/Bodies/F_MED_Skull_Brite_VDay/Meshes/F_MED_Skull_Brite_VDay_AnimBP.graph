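# Post-body animation blueprint for the F_MED_Skull_Brite_VDay female player skin, adding secondary motion over a pose copied from the attached parent mesh.
# Two weighted constraints drive the knife and its aim bone from spine, clavicle and upper-arm bones, then a Look At orients the knife toward the aim bone.
# Spring-based Anim Dynamics sways the bandana, then a partially blended RigidBody node simulates the skin's physics asset.
# The event graph holds no logic.
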AnimBlueprint /Game/Characters/Player/Female/Medium/Bodies/F_MED_Skull_Brite_VDay/Meshes/F_MED_Skull_Brite_VDay_AnimBP
target_skeleton=&/F_MED_Skull_Brite_VDay_Skeleton
graphs [
 AnimGraph {nodes [
   n1 = "Constraint - dyn_knife (PPPP)"!(
    ComponentPose=Local To Component(LocalPose=Copy Pose From Mesh() {bUseAttachedParent=true})
    Alpha=0.675
    ConstraintWeights_0=0.0033
    ConstraintWeights_1:float=0.0136
    ConstraintWeights_2:float=0.0487
    ConstraintWeights_3:float=0.9343
   ) -> (Pose) {
    alphaBoolBlend {blendOption=Linear}
    boneToModify {boneName=dyn_knife}
    constraintSetup [
     {targetBone {boneName=upperarm_twist_01_l}}
     {targetBone {boneName=spine_04}}
     {targetBone {boneName=spine_05}}
     {targetBone {boneName=clavicle_l}}
    ]
    constraintWeights [1,1,1,1]
   }
   n2 = "Constraint - dyn_knife_aim (PPPPP)"!(
    ComponentPose=@n1
    Alpha=1
    ConstraintWeights_0=0.0001
    ConstraintWeights_1:float=0
    ConstraintWeights_2:float=0.7264
    ConstraintWeights_3:float=0.2234
    ConstraintWeights_4:float=0.0501
   ) -> (Pose) {
    alphaBoolBlend {blendOption=Linear}
    boneToModify {boneName=dyn_knife_aim}
    constraintSetup [
     {targetBone {boneName=clavicle_r}}
     {targetBone {boneName=upperarm_twist_01_l}}
     {targetBone {boneName=spine_04}}
     {targetBone {boneName=spine_05}}
     {targetBone {boneName=clavicle_l}}
    ]
    constraintWeights [1,1,1,1,1]
   }
   n3 = Look At - Bone: dyn_knife(ComponentPose=@n2, Alpha=1) {
    alphaBoolBlend {blendOption=Linear}
    boneToModify {boneName=dyn_knife}
    lookAtLocation=<0 0.1 5>
    lookAtTarget {boneReference {boneName=dyn_knife_aim}}
    lookAt_Axis {axis {y=0, z=1}}
    lookUp_Axis {axis=<0 1>}
   }
   n4 = Anim Dynamics - Bone: dyn_bandana(ComponentPose=@n3, Alpha=0.7, ExternalForce="0,0,0") {
    angularBiasOverride=0.9
    angularDampingOverride=0.995
    angularSpringConstant=150
    bAngularSpring=true
    bGravityOverrideInSimSpace=true
    bOverrideAngularBias=true
    bOverrideAngularDamping=true
    gravityScale=0.33
    simulationSpace=RootRelative
    alphaBoolBlend {blendOption=Linear}
    boundBone {boneName=dyn_bandana}
    physicsBodyDefinitions [{
      collisionType=CustomSphere
      sphereCollisionRadius=1
      boundBone {boneName=dyn_bandana}
      boxExtents=<10 10 10>
      constraintSetup {twistAxis=AxisZ, angularLimitsMax=<15 15 10>, angularLimitsMin=<-15 -15 -10>, angularTarget=<0 0 0>}
      localJointOffset=<0 0 -5>
     }]
    relativeSpaceBone {boneName=spine_05}
   }
   n5 = RigidBody(ComponentPose=@n4, Alpha=0.6) {
    bOverrideWorldGravity=true
    overridePhysicsAsset=&/F_MED_Skull_Brite_VDay_Physics
    simulationSpace=BaseBoneSpace
    alphaBoolBlend {blendOption=Linear}
    baseBoneRef {boneName=root}
    overrideWorldGravity {z=-980}
   }
   Output Pose(Result=Component To Local(ComponentPose=@n5)) {layerGroup=None}
  ]}
 EventGraph {nodes [
   TryGetPawnOwner()
   Event BlueprintUpdateAnimation()
  ]}
]
variables []
